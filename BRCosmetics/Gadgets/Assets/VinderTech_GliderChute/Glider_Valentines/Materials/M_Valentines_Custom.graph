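# Masked subsurface material for the Valentines glider cosmetic, combining glittering fresnel-tinted ribbons with a galaxy-style emissive interior.
# A remapped diffuse red channel masks the ribbons, driving their heart sparkle, roughness and subsurface brightness.
# Vertex red gates a camera-vector triplanar nebula, warped by normal-based fresnel and overlaid with rotating and twinkling heart "stars".
# Attributes pass through MF_Fade_Along_UV, using vertex green for opacity-mask fading, then the accessory tech-art override function.
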Material /BRCosmetics/Gadgets/Assets/VinderTech_GliderChute/Glider_Valentines/Materials/M_Valentines_Custom
graph {
 properties {
  bUseMaterialAttributes=true
  bUsedWithMeshDeformer=true
  bUsedWithSkeletalMesh=true
  blendMode=BLEND_Masked
  pixelDepthOffsetMode=PDOM_Legacy
  shadingModel=MSM_Subsurface
  usedShadingModels=Subsurface
 }
 nodes [
  n1 = VertexColor()
  Add(A=@n1.R)
  n2 = TextureSample() {texture=/BRCosmetics/Gadgets/Assets/VinderTech_GliderChute/Glider_Valentines/Textures/T_Glider_Valentines_D}
  n3 = VectorParameter() {parameterName=ColorFresnelRibbons, defaultValue=<rgba 5 1 0.18 1>}
  n4 = VectorParameter() {parameterName=ColorFresnelRibbons2, defaultValue=<rgba 0.099 0.0087 0 1>}
  n5 = Fresnel(ExponentIn=ScalarParameter() {defaultValue=0.03, parameterName=FresnelRibbons}) {baseReflectFraction=0}
  n6 = TextureSample(UVs=Multiply(A=TextureCoordinate(), B=ScalarParameter() {defaultValue=28.4019, parameterName=SparcleRibbons})) {samplerType=SAMPLERTYPE_Masks, texture=/BRCosmetics/Gadgets/Assets/VinderTech_GliderChute/Glider_Valentines/Textures/T_Valentines_Hearts_M}
  n7 = Constant() {r=1}
  n8 = Clamp(None=LinearInterpolate(
    A=ScalarParameter() {defaultValue=-15.0431, parameterName=Ribbonmask1}
    B=ScalarParameter() {defaultValue=8.9261, parameterName=Ribbonmask2}
    Alpha=@n2.R
   ))
  n9 = Multiply(A=Multiply(A=LinearInterpolate(A=@n3.RGB, B=@n4.RGB, Alpha=@n5), B=Add(A=@n6.R, B=@n7)), B=@n8)
  n10 = ScalarParameter() {defaultValue=3.8658, parameterName=Fresnel1}
  n11 = TextureSample() {texture=/BRCosmetics/Gadgets/Assets/VinderTech_GliderChute/Glider_Valentines/Textures/T_Glider_Valentines_E}
  n12 = Add(
   A=Multiply(A=Fresnel(ExponentIn=@n10) {baseReflectFraction=0}, B=ScalarParameter() {defaultValue=0.2476, parameterName=FresnelBrightness})
   B=Multiply(A=@n11.RGB, B=ScalarParameter() {defaultValue=0.6762, parameterName=Emissive})
  )
  n13 = Add(A=@n2.RGB, B=Add(A=@n9, B=@n12))
  n14 = TextureSample() {samplerType=SAMPLERTYPE_Masks, texture=/BRCosmetics/Gadgets/Assets/VinderTech_GliderChute/Glider_Valentines/Textures/T_Glider_Valentines_S}
  n15 = LinearInterpolate(A=@n14.B, B=ScalarParameter() {defaultValue=0.3429, parameterName=Ribbon}, Alpha=@n8)
  n16 = Fresnel(ExponentIn=ScalarParameter() {defaultValue=0.5457, parameterName=Fresnel2}) {baseReflectFraction=0}
  n17 = ScalarParameter() {defaultValue=1.5646, group=Galaxy, parameterName=GalaxyTiling}
  n18 = TextureSample() {samplerType=SAMPLERTYPE_Normal, texture=/BRCosmetics/Gadgets/Assets/VinderTech_GliderChute/Glider_Valentines/Textures/T_Glider_Valentines_N}
  n19 = Multiply(
   A=Fresnel(ExponentIn=ScalarParameter() {defaultValue=1.7031, parameterName=WarpFresnel}, Normal=Transform(None=@n18.RGB))
   B=ScalarParameter() {defaultValue=-0.439, group=Galaxy, parameterName=WarpIntensity}
  )
  n20 = ScalarParameter() {defaultValue=3, parameterName=asdf}
  n21 = Time()
  n22 = Constant()
  n23 = ScalarParameter() {defaultValue=0.02, group=Galaxy, parameterName=Galaxy Rotation Speed}
  n24 = ScalarParameter() {defaultValue=-0.02, group=Galaxy, parameterName=Galaxy Rotation Speed_1}
  n25 = Multiply(
   A=LinearInterpolate(A=@n21, B=Multiply(A=@n21, B=@n22), Alpha=ScalarParameter() {group=Galaxy, parameterName=Background Rotation})
   B=AppendVector(A=AppendVector(A=@n23, B=@n24), B=ScalarParameter() {defaultValue=0.1, group=Galaxy, parameterName=Galaxy Rotation Speed_2})
  )
  n26 = TextureObject() {texture=/Game/Characters/Player/Male/Medium/Bodies/M_Med_Soldier_04/Skins/Galaxy/Textures/T_FN_Nebula}
  n27 = Desaturation(None=@n11.RGB, Fraction=ScalarParameter() {defaultValue=0.1333, parameterName=DesatSmallHearts})
  n28 = VectorParameter() {group=Galaxy, parameterName=Star Brightness, defaultValue=<rgba 100 100 100 1>}
  n29 = Time()
  n30 = ScalarParameter() {defaultValue=-0.125, group=Galaxy, parameterName=Small Star Rotation speed}
  n31 = TextureObject() {samplerType=SAMPLERTYPE_Masks, texture=&T_Valentines_Hearts_M}
  n32 = ComponentMask(None=MaterialFunctionCall!(
    CameraVector
    Rotation Speed=Multiply(A=@n29, B=@n30)
    Texture=@n31
    Tiling=ScalarParameter() {defaultValue=1.5, group=Galaxy, parameterName=Small Star Tiling}
    Offset
    AxisFadeContrast
   ) {materialFunction=/Game/Weapons/WeaponSkins/MaterialFunction/TriplanarCameraVector_Rotated}.Result) {b=true}
  n33 = TextureObject() {samplerType=SAMPLERTYPE_Masks, texture=&T_Valentines_Hearts_M}
  n34 = ComponentMask(None=MaterialFunctionCall!(
    CameraVector
    Texture=@n33
    Tiling=ScalarParameter() {defaultValue=2, group=Galaxy, parameterName=Bright Star Tiling}
    Offset
    AxisFadeContrast
   ) {materialFunction=/Engine/Functions/Engine_MaterialFunctions01/Texturing/TriplanarCameraVector}.Result) {r=true}
  n35 = TextureSample(UVs=Panner(Coordinate=TextureCoordinate() {uTiling=0.25, vTiling=0.25}, Speed=ScalarParameter() {defaultValue=0.0381, parameterName=SpeedTwinkle}) {speedX=0.133, speedY=0.2}) {
   samplerSource=SSM_Wrap_WorldGroupSettings
   samplerType=SAMPLERTYPE_Masks
   texture=&T_Valentines_Hearts_M
  }
  n36 = LinearInterpolate(
   A=LinearInterpolate(
    A=MaterialFunctionCall(Rotation Speed=@n25, Texture=@n26, Tiling=@n17, Offset=@n19, AxisFadeContrast=@n20) {materialFunction=&TriplanarCameraVector_Rotated}.Result
    B=Multiply(A=@n27, B=@n28.RGB)
    Alpha=Multiply(A=@n32, B=@n32)
   )
   B=Multiply(A=@n27, B=VectorParameter() {group=Galaxy, parameterName=Star Brightness_1, defaultValue=<rgba 500 500 500 1>}.RGB)
   Alpha=Multiply(A=Multiply(A=@n34, B=@n34), B=@n35.G)
  )
  n37 = Multiply(A=@n1.R, B=Add(A=@n12, B=Multiply(A=@n16, B=@n36)))
  n38 = Add(A=Multiply(A=@n9, B=ScalarParameter() {defaultValue=0.4095, parameterName=RibbonEmissive}), B=@n37)
  n39 = ScalarParameter() {defaultValue=0.3667, parameterName=SSSDistance}
  n40 = ScalarParameter() {defaultValue=1, parameterName=SSSBrightness}
  n41 = ScalarParameter() {defaultValue=0.25, parameterName=SSSBrightnessRibbon}
  n42 = Multiply(A=Multiply(A=Add(A=@n13, B=@n37), B=LinearInterpolate(A=@n40, B=@n41, Alpha=@n8)), B=Add(A=@n8, B=@n1.R))
  MaterialFunctionCall(In=MaterialFunctionCall!(
    Material=MakeMaterialAttributes!(
     BaseColor=@n13
     Metallic=@n14.G
     Specular=@n14.G
     Roughness=@n15
     Anisotropy
     EmissiveColor=@n38
     Opacity=@n39
     OpacityMask
     Normal=@n18.RGB
     Tangent
     WorldPositionOffset
     SubsurfaceColor=@n42
     ClearCoat
     ClearCoatRoughness
     AmbientOcclusion
     Refraction
     CustomizedUVs_0
     CustomizedUVs_1
     CustomizedUVs_2
     CustomizedUVs_3
     CustomizedUVs_4
     CustomizedUVs_5
     CustomizedUVs_6
     CustomizedUVs_7
     PixelDepthOffset
     ShadingModel
     Displacement
    )
    Mask for Fade=VertexColor().G
   ) {materialFunction=/Game/Packages/Fortress_SharedMaterials/Base_Material/MF_Fade_Along_UV}.Fade in OpacityMask) {materialFunction=&MF_TechArtOverrides_accessories}
 ]
}
